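# Material function that shades the arid, sandy tile type of the temperate hexmap zone, outputting Diffuse, Normals, sand height and sand texture.
# A swirl flowmap distorts world-XY UVs for a sand sample and a mask that blends in rocky desert sand.
# Vertex red blends dune color into sand and gates transformed normal-map detail; green multiplies diffuse.
# HeightLerp combines TileTypeMask with sharpened sand alpha to weight Diffuse and Normals.
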MaterialFunction /Game/Environments/Hexmap/Zones/Temperate/Arid/MF_TileType-Arid
graph {nodes [
  n1 = Constant() {r=2}
  n2 = Constant() {r=-0.0001}
  n3 = ComponentMask(None=WorldPosition().XYZ) {g=true, r=true}
  n4 = FunctionInput(Preview=Multiply(A=Multiply(A=@n2, B=@n3), B=Constant() {r=17})) {inputName=Worldposition, inputType=FunctionInput_Vector2}
  n5 = Multiply(A=@n4, B=Constant() {r=2})
  n6 = TextureSample(UVs=Multiply(A=Constant() {r=0.3}, B=@n5)) {
   samplerSource=SSM_Wrap_WorldGroupSettings
   samplerType=SAMPLERTYPE_LinearColor
   texture=/Game/VisualThreatManager/StormVisuals/LargeClouds/T_Swirl_Flowmap
  }
  n7 = ComponentMask(None=@n6.RGB) {g=true, r=true}
  n8 = Constant() {r=0.8}
  n9 = TextureSample(UVs=Add(A=@n5, B=Multiply(A=@n7, B=@n8))) {samplerSource=SSM_Wrap_WorldGroupSettings}
  n10 = Power(Base=Clamp(None=Multiply(A=@n1, B=@n9.A)), Exp=Constant() {r=10})
  n11 = MaterialFunctionCall(Transition Phase=FunctionInput() {inputName=TileTypeMask, inputType=FunctionInput_Scalar, previewValue=<1 1 1 1>}, Height Texture=@n10) {materialFunction=&HeightLerp}
  n12 = Add(A=Multiply(A=@n7, B=Constant() {r=0.5}), B=Multiply(A=@n5, B=Constant() {r=0.3}))
  n13 = Clamp(None=Multiply(
    A=TextureSample(UVs=@n12) {samplerSource=SSM_Wrap_WorldGroupSettings}.R
    B=ScalarParameter() {defaultValue=1.5, group=TileType-Arid, parameterName=ModulateSandMask}
   ))
  n14 = VertexColor()
  n15 = Multiply(A=@n13, B=@n14.R)
  n16 = Multiply(
   A=TextureSample(UVs=@n4) {samplerSource=SSM_Wrap_WorldGroupSettings, texture=/Game/Environments/Sets/Arid/Textures/T_desert_sand_rocky_C}.RGB
   B=ScalarParameter() {defaultValue=0.75, group=TileType-Arid, parameterName=SandBright_1}
  )
  n17 = Multiply(A=@n9.RGB, B=ScalarParameter() {defaultValue=1.5, group=TileType-Arid, parameterName=SandBright})
  n18 = LinearInterpolate(
   A=VectorParameter() {group=TileType-Arid, parameterName=SandDuneColor, defaultValue=<rgba 0.6354 0.4579 0.1945 1>}.RGB
   B=LinearInterpolate(A=@n17, B=@n16, Alpha=@n15)
   Alpha=@n14.R
  )
  FunctionOutput(None=Multiply(A=@n11.Lerp Alpha No Contrast, B=Multiply(A=@n18, B=@n14.G))) {outputName=Diffuse}
  n19 = VertexNormalWS()
  n20 = VectorParameter() {group=TileType-Arid, parameterName=AridNormalAmount, defaultValue=<rgba 2 2 1 1>}
  n21 = TextureSample(UVs=@n4) {
   samplerSource=SSM_Wrap_WorldGroupSettings
   samplerType=SAMPLERTYPE_Normal
   texture=/Game/Environments/Sets/Arid/Textures/T_desert_sand_rocky_N
  }
  n22 = VertexNormalWS()
  n23 = If(
   A=DotProduct(A=@n22, B=Constant3Vector() {constant=<rgba 0 0 1 0>}) {desc=generate sign for this projection}
   B=Constant()
   "A > B"=Constant() {r=1}
   "A < B"=Constant() {r=-1}
  )
  n24 = Constant() {r=-1}
  n25 = Constant() {r=1}
  n26 = MaterialFunctionCall!(Vector1=@n22, Vector2=Constant3Vector() {constant=<rgba 0 1 0 0>}) {materialFunction=/Engine/Functions/Engine_MaterialFunctions02/Math/CreateThirdOrthogonalVector}
  n27 = MaterialFunctionCall(
   VectorToTransform=Multiply(
    A=Multiply(A=@n20.RGB, B=@n21.RGB)
    B=MaterialFunctionCall(X=@n23, Y=@n24, Z=@n25) {materialFunction=&MakeFloat3}.Result
   )
   BasisX=@n26.Vector3
   BasisY=@n26.Vector3
   BasisZ=@n26.Vector3
  ) {
   desc=lerps in when the vertex normal faces X or Negative X
   materialFunction=&Transform3x3Matrix
  }
  FunctionOutput(None=Multiply(A=@n11.Lerp Alpha No Contrast, B=LinearInterpolate(A=@n19, B=@n27.Transformed Vector, Alpha=@n15))) {outputName=Normals}
  FunctionOutput(None=@n10) {outputName=SandTextureHeight, sortPriority=1}
  FunctionOutput(None=LinearInterpolate(A=@n17, B=@n16, Alpha=@n13)) {outputName=SandTexture, sortPriority=1}
 ]}
